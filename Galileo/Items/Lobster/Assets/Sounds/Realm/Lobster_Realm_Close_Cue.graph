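# Sound cue for closing the Lobster item's "Realm", classed as weapon foley and using the Lobster melee equip attenuation.
# A Random node picks one of two equally weighted close waves, excluded from branch culling, and feeds a Modulator that slightly varies pitch.
# The cue plays at raised base pitch and very low volume, with a concurrency override that caps simultaneous plays and stops the lowest priority.
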SoundCue /Galileo/Items/Lobster/Assets/Sounds/Realm/Lobster_Realm_Close_Cue
root=Modulator_0
nodes [
 {id=Modulator_0, children [Random_1], properties {pitchMax=1.1, pitchMin=0.9}}
 {id=Random_1, children [WavePlayer_0,WavePlayer_3], properties {bShouldExcludeFromBranchCulling=true, weights [1,1]}}
 {id=WavePlayer_0, properties {soundWaveAssetPtr=&/Lobster_Realm_Close_02}}
 {id=WavePlayer_3, properties {soundWaveAssetPtr=&/Lobster_Realm_Close_01}}
]
properties {
 attenuationSettings=/Game/Sounds/Attenuation/Fort_Weapon_Melee_Lobster_Equip_Att
 bOverrideConcurrency=true
 maxDistance=4000
 pitchMultiplier=1.2
 soundClassObject=&Fort_WeaponFoley
 volumeMultiplier=0.056
 concurrencyOverrides {maxCount=2, resolutionRule=StopLowestPriority}
}
