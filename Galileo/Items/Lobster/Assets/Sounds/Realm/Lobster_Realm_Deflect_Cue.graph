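# Sound cue for the deflect sound of the Lobster melee weapon's Realm variant, with a poise warning layer.
# A mixer layers randomized deflect waves, pitch-modulated by InvPoise, with a randomized poise warning that crossfades in at high InvPoise under its own pitch modulation.
# A quality switch gives the lower two levels the full deflect pool and the upper two a four-wave subset.
# Concurrency is capped, stopping lowest priority.
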SoundCue /Galileo/Items/Lobster/Assets/Sounds/Realm/Lobster_Realm_Deflect_Cue
root=Mixer_0
nodes [
 {id=Mixer_0, children [ModulatorContinuous_1, ParamCrossFade_1], properties {inputVolume [1,1]}}
 {
  id=ModulatorContinuous_1
  children [QualityLevel_0]
  properties {pitchModulationParams {default=0, maxOutput=1.2, minOutput=0.8, parameterName=InvPoise}}
 }
 {
  id=ModulatorContinuous_2
  children [Random_0]
  properties {pitchModulationParams {default=0, maxOutput=1.5, minOutput=0.9, parameterName=InvPoise}}
 }
 {
  id=ParamCrossFade_1
  children [ModulatorContinuous_2]
  properties {
   paramName=InvPoise
   crossFadeInput [{fadeInDistanceEnd=1, fadeInDistanceStart=0.7, fadeOutDistanceEnd=100, fadeOutDistanceStart=1, volume=1.2}]
  }
 }
 {id=QualityLevel_0, children [Random_2,Random_2,Random_1,Random_1]}
 {id=Random_0, children [WavePlayer_10, WavePlayer_11], properties {weights [1,1]}}
 {
  id=Random_1
  children [WavePlayer_24, WavePlayer_25, WavePlayer_26, WavePlayer_27]
  properties {bShouldExcludeFromBranchCulling=true, weights [1,1,1,1]}
 }
 {
  id=Random_2
  children [
   WavePlayer_12
   WavePlayer_13
   WavePlayer_14
   WavePlayer_15
   WavePlayer_22
   WavePlayer_23
   WavePlayer_24
   WavePlayer_25
   WavePlayer_26
   WavePlayer_27
  ]
  properties {weights [1,1,1,1,1,1,1,1,1,1]}
 }
 {id=WavePlayer_10, properties {soundWaveAssetPtr=&/Lobster_Realm_Poise_Warning_01}}
 {id=WavePlayer_11, properties {soundWaveAssetPtr=&/Lobster_Realm_Poise_Warning_02}}
 {id=WavePlayer_12, properties {soundWaveAssetPtr=&/Lobster_Realm_Deflect_01}}
 {id=WavePlayer_13, properties {soundWaveAssetPtr=&/Lobster_Realm_Deflect_02}}
 {id=WavePlayer_14, properties {soundWaveAssetPtr=&/Lobster_Realm_Deflect_03}}
 {id=WavePlayer_15, properties {soundWaveAssetPtr=&/Lobster_Realm_Deflect_04}}
 {id=WavePlayer_22, properties {soundWaveAssetPtr=&/Lobster_Realm_Deflect_05}}
 {id=WavePlayer_23, properties {soundWaveAssetPtr=&/Lobster_Realm_Deflect_06}}
 {id=WavePlayer_24, properties {soundWaveAssetPtr=&/Lobster_Realm_Deflect_07}}
 {id=WavePlayer_25, properties {soundWaveAssetPtr=&/Lobster_Realm_Deflect_08}}
 {id=WavePlayer_26, properties {soundWaveAssetPtr=&/Lobster_Realm_Deflect_09}}
 {id=WavePlayer_27, properties {soundWaveAssetPtr=&/Lobster_Realm_Deflect_10}}
]
properties {
 attenuationSettings=/Game/Sounds/Attenuation/Fort_Weapon_Melee_Lobster_Attack_Deflect_Stereo_Att
 bOverrideConcurrency=true
 maxDistance=12000
 soundClassObject=&Fort_WeaponMelee
 volumeMultiplier=0.8
 concurrencyOverrides {maxCount=3, resolutionRule=StopLowestPriority}
}
