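# Sound cue for the looping idle sound of the Lobster Realm item, classed as a melee weapon sound with close-range attenuation.
# A looping wave player feeds a neutral modulator into a root enveloper.
# The enveloper shifts pitch over the first half second while fading volume in, then eases the volume down over the following seconds.
# Concurrency is overridden to cap simultaneous instances, stopping the lowest priority.
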SoundCue /Galileo/Items/Lobster/Assets/Sounds/Realm/Lobster_Realm_IdleLoop_Cue
root=Enveloper_1
nodes [
 {
  id=Enveloper_1
  children [Modulator_0]
  properties {
   pitchCurve {editorCurveData {keys [<at 0 1.1>, {time=0.5}]}}
   volumeCurve {editorCurveData {keys [
      <at 0 0 RCIM_Cubic>
      {arriveTangent=-0.0839, interpMode=RCIM_Cubic, leaveTangent=-0.0839, tangentMode=RCTM_User, time=0.5, value=1}
      <at 10 0.5>
     ]}}
  }
 }
 {id=Modulator_0, children [WavePlayer_0], properties {pitchMax=1, pitchMin=1, volumeMax=1, volumeMin=1}}
 {id=WavePlayer_0, properties {bLooping=true, soundWaveAssetPtr=&/Lobster_Realm_Loop}}
]
properties {
 attenuationSettings=/Game/Sounds/Attenuation/Fort_Weapon_Melee_Lobster_Idle_Loop_Close_Att
 bOverrideConcurrency=true
 maxDistance=3000
 soundClassObject=&Fort_WeaponMelee
 volumeMultiplier=0.25
 concurrencyOverrides {maxCount=3, resolutionRule=StopLowestPriority}
}
