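# Sound cue for the Lobster item's realm-open effect, in the melee weapon sound class with a small concurrency limit.
# A mixer layers two separately attenuated branches: a briefly delayed random pick of three open one-shots with pitch and volume variation, and a quieter looping realm sound.
# The loop starts after a short delay, and an enveloper fades it in while easing its pitch down, then tapers its volume over several seconds.
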SoundCue /Galileo/Items/Lobster/Assets/Sounds/Realm/Lobster_Realm_Open_Cue
root=Mixer_1
nodes [
 {
  id=Attenuation_0
  children [Enveloper_1]
  properties {attenuationSettings=/Game/Sounds/Attenuation/Fort_Weapon_Melee_Lobster_Idle_Loop_Close_Att}
 }
 {id=Attenuation_1, children [Modulator_2], properties {attenuationSettings=/Game/Sounds/Attenuation/Fort_Weapon_Melee_Lobster_Equip_Att}}
 {id=Delay_0, children [Modulator_0], properties {delayMax=0.25, delayMin=0.25}}
 {id=Delay_1, children [Random_1], properties {delayMax=0.1, delayMin=0.1}}
 {
  id=Enveloper_1
  children [Delay_0]
  properties {
   pitchCurve {editorCurveData {keys [<at 0 1.1>, {time=0.5}]}}
   volumeCurve {editorCurveData {keys [
      <at 0 0 RCIM_Cubic>
      {arriveTangent=-0.0839, interpMode=RCIM_Cubic, leaveTangent=-0.0839, tangentMode=RCTM_User, time=0.5, value=1}
      <at 10 0.453>
     ]}}
  }
 }
 {id=Mixer_1, children [Attenuation_1, Attenuation_0], properties {inputVolume [1,0.4]}}
 {id=Modulator_0, children [WavePlayer_5], properties {pitchMax=1, pitchMin=1, volumeMax=1, volumeMin=1}}
 {id=Modulator_2, children [Delay_1], properties {pitchMax=1.2, volumeMax=1, volumeMin=0.7}}
 {id=Random_1, children [WavePlayer_0,WavePlayer_1,WavePlayer_2], properties {weights [1,1,1]}}
 {id=WavePlayer_0, properties {soundWaveAssetPtr=&/Lobster_Realm_Open_01}}
 {id=WavePlayer_1, properties {soundWaveAssetPtr=&/Lobster_Realm_Open_02}}
 {id=WavePlayer_2, properties {soundWaveAssetPtr=&/Lobster_Realm_Open_03}}
 {id=WavePlayer_5, properties {bLooping=true, soundWaveAssetPtr=&/Lobster_Realm_Loop}}
]
properties {
 bOverrideConcurrency=true
 maxDistance=4000
 soundClassObject=&Fort_WeaponMelee
 volumeMultiplier=0.4217
 concurrencyOverrides {maxCount=3, resolutionRule=StopLowestPriority}
}
