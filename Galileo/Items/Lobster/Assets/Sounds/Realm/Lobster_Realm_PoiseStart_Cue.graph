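# Sound cue for the start of the Lobster item's Realm "Poise" action, classed as a melee weapon sound.
# A Random node picks among four equally weighted Poise wave variants, and a Modulator adds slight pitch and volume variation.
# It uses close-range mono melee attack attenuation at reduced volume.
# A concurrency override allows only a couple of simultaneous instances, stopping the lowest priority one when the limit is hit.
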SoundCue /Galileo/Items/Lobster/Assets/Sounds/Realm/Lobster_Realm_PoiseStart_Cue
root=Modulator_0
nodes [
 {id=Modulator_0, children [Random_0], properties {pitchMax=1.1, volumeMin=0.8}}
 {
  id=Random_0
  children [WavePlayer_1,WavePlayer_2,WavePlayer_3,WavePlayer_6]
  properties {bShouldExcludeFromBranchCulling=true, weights [1,1,1,1]}
 }
 {id=WavePlayer_1, properties {soundWaveAssetPtr=&/Lobster_Realm_Poise_01}}
 {id=WavePlayer_2, properties {soundWaveAssetPtr=&/Lobster_Realm_Poise_02}}
 {id=WavePlayer_3, properties {soundWaveAssetPtr=&/Lobster_Realm_Poise_03}}
 {id=WavePlayer_6, properties {soundWaveAssetPtr=&/Lobster_Realm_Poise_04}}
]
properties {
 attenuationSettings=/Game/Sounds/Attenuation/Fort_Weapon_Melee_Attack_Close_Range_Mono_Att
 bOverrideConcurrency=true
 maxDistance=10058.75
 soundClassObject=&Fort_WeaponMelee
 volumeMultiplier=0.2812
 concurrencyOverrides {maxCount=2, resolutionRule=StopLowestPriority}
}
