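# Sound cue for catching the thrown Lobster melee weapon on recall, in its Realm variant.
# A mixer layers a random default swing, a pitch-modulated Realm catch sound and a slightly delayed revolver pickup sound.
# A quality switch gates the pickup layer, leaving the fourth quality slot empty.
# The cue uses melee catch attenuation and overrides concurrency to allow a single instance.
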SoundCue /Galileo/Items/Lobster/Assets/Sounds/Realm/Lobster_Realm_Throw_Catch_Cue
root=Mixer_0
nodes [
 {id=Delay_0, children [Modulator_0]}
 {id=Delay_1, children [QualityLevel_1], properties {delayMax=0.1, delayMin=0.1}}
 {id=Mixer_0, children [Delay_0,Modulator_1,Delay_1], properties {inputVolume [1.1,1,0.5]}}
 {id=Modulator_0, children [Random_3]}
 {id=Modulator_1, children [WavePlayer_0], properties {pitchMax=1.3}}
 {id=Modulator_2, children [WavePlayer_2]}
 {id=QualityLevel_1, children [Modulator_2,Modulator_2,Modulator_2,null]}
 {
  id=Random_3
  children [WavePlayer_16, WavePlayer_17, WavePlayer_18, WavePlayer_19, WavePlayer_20, WavePlayer_21, WavePlayer_22, WavePlayer_23]
  properties {bShouldExcludeFromBranchCulling=true, weights [1,1,1,1,1,1,1,1]}
 }
 {id=WavePlayer_0, properties {soundWaveAssetPtr=&/Lobster_Realm_Throw_Prj_Recall_Catch_01}}
 {id=WavePlayer_16, properties {soundWaveAssetPtr=/Galileo/Items/Lobster/Assets/Sounds/Default/Lobster_Default_Swing_08}}
 {id=WavePlayer_17, properties {soundWaveAssetPtr=/Galileo/Items/Lobster/Assets/Sounds/Default/Lobster_Default_Swing_01}}
 {id=WavePlayer_18, properties {soundWaveAssetPtr=/Galileo/Items/Lobster/Assets/Sounds/Default/Lobster_Default_Swing_02}}
 {id=WavePlayer_19, properties {soundWaveAssetPtr=/Galileo/Items/Lobster/Assets/Sounds/Default/Lobster_Default_Swing_03}}
 {id=WavePlayer_2, properties {soundWaveAssetPtr=/Game/Sounds/Fort_GamePlay_Sounds/Loot/Revolver_Pickup}}
 {id=WavePlayer_20, properties {soundWaveAssetPtr=/Galileo/Items/Lobster/Assets/Sounds/Default/Lobster_Default_Swing_04}}
 {id=WavePlayer_21, properties {soundWaveAssetPtr=/Galileo/Items/Lobster/Assets/Sounds/Default/Lobster_Default_Swing_05}}
 {id=WavePlayer_22, properties {soundWaveAssetPtr=/Galileo/Items/Lobster/Assets/Sounds/Default/Lobster_Default_Swing_06}}
 {id=WavePlayer_23, properties {soundWaveAssetPtr=/Galileo/Items/Lobster/Assets/Sounds/Default/Lobster_Default_Swing_07}}
]
properties {
 attenuationSettings=/Game/Sounds/Attenuation/Fort_Weapon_Melee_Lobster_Prj_Catch_Att
 bOverrideConcurrency=true
 maxDistance=2000
 soundClassObject=&Fort_WeaponMelee
 volumeMultiplier=0.35
 concurrencyOverrides {maxCount=1, resolutionRule=StopLowestPriority}
}
