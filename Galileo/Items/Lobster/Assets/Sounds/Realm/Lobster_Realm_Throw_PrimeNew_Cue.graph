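# Sound cue for priming a throw of the Lobster item in its Realm variant, built from Default wave assets and played as a melee weapon sound with special-attack attenuation.
# A mixer layers a random swing with a louder, randomly pitch-modulated sweetener.
# A quality switch gives the first two levels eight pitch-enveloped swings and the last two a plain four-swing subset, with both swing pools excluded from branch culling.
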SoundCue /Galileo/Items/Lobster/Assets/Sounds/Realm/Lobster_Realm_Throw_PrimeNew_Cue
root=Mixer_0
nodes [
 {
  id=Enveloper_0
  children [Random_3]
  properties {pitchCurve {editorCurveData {keys [{time=-0.0055}, {time=0}, <at 0.0441 1.2>, {time=0.5}]}}, volumeCurve {editorCurveData {keys [<at 0 1>]}}}
 }
 {id=Mixer_0, children [QualityLevel_0, Modulator_0], properties {inputVolume [1,1.75]}}
 {id=Modulator_0, children [Random_0], properties {pitchMax=1.1, pitchMin=0.8}}
 {id=QualityLevel_0, children [Enveloper_0,Enveloper_0,Random_4,Random_4]}
 {id=Random_0, children [WavePlayer_0,WavePlayer_1,WavePlayer_2], properties {weights [1,1,1]}}
 {
  id=Random_3
  children [WavePlayer_16, WavePlayer_17, WavePlayer_18, WavePlayer_19, WavePlayer_20, WavePlayer_21, WavePlayer_22, WavePlayer_23]
  properties {bShouldExcludeFromBranchCulling=true, weights [1,1,1,1,1,1,1,1]}
 }
 {
  id=Random_4
  children [WavePlayer_24, WavePlayer_25, WavePlayer_26, WavePlayer_27]
  properties {bShouldExcludeFromBranchCulling=true, weights [1,1,1,1]}
 }
 {id=WavePlayer_0, properties {soundWaveAssetPtr=/Galileo/Items/Lobster/Assets/Sounds/Default/Lobster_Default_Throw_Prime_Sweetener_03}}
 {id=WavePlayer_1, properties {soundWaveAssetPtr=/Galileo/Items/Lobster/Assets/Sounds/Default/Lobster_Default_Throw_Prime_Sweetener_02}}
 {id=WavePlayer_16, properties {soundWaveAssetPtr=/Galileo/Items/Lobster/Assets/Sounds/Default/Lobster_Default_Swing_08}}
 {id=WavePlayer_17, properties {soundWaveAssetPtr=/Galileo/Items/Lobster/Assets/Sounds/Default/Lobster_Default_Swing_01}}
 {id=WavePlayer_18, properties {soundWaveAssetPtr=/Galileo/Items/Lobster/Assets/Sounds/Default/Lobster_Default_Swing_02}}
 {id=WavePlayer_19, properties {soundWaveAssetPtr=/Galileo/Items/Lobster/Assets/Sounds/Default/Lobster_Default_Swing_03}}
 {id=WavePlayer_2, properties {soundWaveAssetPtr=/Galileo/Items/Lobster/Assets/Sounds/Default/Lobster_Default_Throw_Prime_Sweetener_01}}
 {id=WavePlayer_20, properties {soundWaveAssetPtr=/Galileo/Items/Lobster/Assets/Sounds/Default/Lobster_Default_Swing_04}}
 {id=WavePlayer_21, properties {soundWaveAssetPtr=/Galileo/Items/Lobster/Assets/Sounds/Default/Lobster_Default_Swing_05}}
 {id=WavePlayer_22, properties {soundWaveAssetPtr=/Galileo/Items/Lobster/Assets/Sounds/Default/Lobster_Default_Swing_06}}
 {id=WavePlayer_23, properties {soundWaveAssetPtr=/Galileo/Items/Lobster/Assets/Sounds/Default/Lobster_Default_Swing_07}}
 {id=WavePlayer_24, properties {soundWaveAssetPtr=&Lobster_Default_Swing_08}}
 {id=WavePlayer_25, properties {soundWaveAssetPtr=&Lobster_Default_Swing_01}}
 {id=WavePlayer_26, properties {soundWaveAssetPtr=&Lobster_Default_Swing_02}}
 {id=WavePlayer_27, properties {soundWaveAssetPtr=&Lobster_Default_Swing_03}}
]
properties {
 attenuationSettings=&/Realm_Special_Attack_Att
 maxDistance=16000
 soundClassObject=&Fort_WeaponMelee
 volumeMultiplier=1
}
