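# Sound cue for the Lobster item's Realm thrown projectile hitting a player, built from Default Lobster impact and swing waves.
# A shared attenuated layer mixes a random impact with two pitch-varied swing layers, one slightly delayed.
# An IsLocal branch adds a low-pitched impact with a fading envelope and its own attenuation for the local case; otherwise only the shared layer plays.
# Quality-level nodes switch the swing layers between larger and smaller variation pools.
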SoundCue /Galileo/Items/Lobster/Assets/Sounds/Realm/Lobster_Realm_Throw_Prj_Hit_Player_Cue
root=Branch_0
nodes [
 {
  id=Attenuation_0
  children [Mixer_0]
  properties {attenuationSettings=/Game/Sounds/Attenuation/Fort_Weapon_Melee_Attack_Close_Range_Stereo_Att}
 }
 {
  id=Attenuation_1
  children [Enveloper_0]
  properties {attenuationSettings=&/SA_Lobster_Realm_Throw_Prj_Hit_Player_1PLayer}
 }
 {id=Branch_0, children [Mixer_1, Attenuation_0, Attenuation_0], properties {boolParameterName=IsLocal}}
 {id=Delay_0, children [Modulator_1], properties {delayMax=0.15, delayMin=0.15}}
 {id=Delay_1, children [Modulator_2]}
 {
  id=Enveloper_0
  children [Modulator_3]
  properties {pitchCurve {editorCurveData {keys [{time=0}, {time=0.2}]}}, volumeCurve {editorCurveData {keys [<at -0.0098 1>,<at 0.5 0>]}}}
 }
 {id=Mixer_0, children [Modulator_0,Delay_1,Delay_0], properties {inputVolume [1,0.7,0.6]}}
 {id=Mixer_1, children [Attenuation_1, Attenuation_0], properties {inputVolume [0.8,1]}}
 {id=Modulator_0, children [Random_1], properties {pitchMax=1.1}}
 {id=Modulator_1, children [QualityLevel_1], properties {pitchMax=1.1}}
 {id=Modulator_2, children [QualityLevel_0], properties {pitchMax=1.3, pitchMin=1.2}}
 {id=Modulator_3, children [Random_6], properties {pitchMax=0.65, pitchMin=0.6, volumeMax=1, volumeMin=1}}
 {id=QualityLevel_0, children [Random_3,Random_3,Random_4,Random_4]}
 {id=QualityLevel_1, children [Random_0,Random_0,Random_2,Random_2]}
 {
  id=Random_0
  children [WavePlayer_1, WavePlayer_2, WavePlayer_3, WavePlayer_4, WavePlayer_5, WavePlayer_6, WavePlayer_12, WavePlayer_13]
  properties {bShouldExcludeFromBranchCulling=true, weights [1,1,1,1,1,1,1,1]}
 }
 {
  id=Random_1
  children [WavePlayer_0, WavePlayer_7, WavePlayer_8, WavePlayer_9, WavePlayer_10, WavePlayer_11]
  properties {bShouldExcludeFromBranchCulling=true, weights [1,1,1,1,1,1]}
 }
 {
  id=Random_2
  children [WavePlayer_14, WavePlayer_15, WavePlayer_28, WavePlayer_29]
  properties {bShouldExcludeFromBranchCulling=true, weights [1,1,1,1]}
 }
 {
  id=Random_3
  children [WavePlayer_16, WavePlayer_17, WavePlayer_18, WavePlayer_19, WavePlayer_20, WavePlayer_21, WavePlayer_22, WavePlayer_23]
  properties {bShouldExcludeFromBranchCulling=true, weights [1,1,1,1,1,1,1,1]}
 }
 {
  id=Random_4
  children [WavePlayer_24, WavePlayer_25, WavePlayer_26, WavePlayer_27]
  properties {bShouldExcludeFromBranchCulling=true, weights [1,1,1,1]}
 }
 {id=Random_6, children [WavePlayer_30, WavePlayer_31, WavePlayer_32], properties {weights [1,1,1]}}
 {id=WavePlayer_0, properties {soundWaveAssetPtr=/Galileo/Items/Lobster/Assets/Sounds/Default/Lobster_Default_Impact_01}}
 {id=WavePlayer_1, properties {soundWaveAssetPtr=/Galileo/Items/Lobster/Assets/Sounds/Default/Lobster_Default_Swing_08}}
 {id=WavePlayer_10, properties {soundWaveAssetPtr=/Galileo/Items/Lobster/Assets/Sounds/Default/Lobster_Default_Impact_05}}
 {id=WavePlayer_11, properties {soundWaveAssetPtr=/Galileo/Items/Lobster/Assets/Sounds/Default/Lobster_Default_Impact_06}}
 {id=WavePlayer_12, properties {soundWaveAssetPtr=/Galileo/Items/Lobster/Assets/Sounds/Default/Lobster_Default_Swing_06}}
 {id=WavePlayer_13, properties {soundWaveAssetPtr=/Galileo/Items/Lobster/Assets/Sounds/Default/Lobster_Default_Swing_07}}
 {id=WavePlayer_14, properties {soundWaveAssetPtr=&Lobster_Default_Swing_08}}
 {id=WavePlayer_15, properties {soundWaveAssetPtr=/Galileo/Items/Lobster/Assets/Sounds/Default/Lobster_Default_Swing_01}}
 {id=WavePlayer_16, properties {soundWaveAssetPtr=&Lobster_Default_Swing_08}}
 {id=WavePlayer_17, properties {soundWaveAssetPtr=&Lobster_Default_Swing_01}}
 {id=WavePlayer_18, properties {soundWaveAssetPtr=/Galileo/Items/Lobster/Assets/Sounds/Default/Lobster_Default_Swing_02}}
 {id=WavePlayer_19, properties {soundWaveAssetPtr=/Galileo/Items/Lobster/Assets/Sounds/Default/Lobster_Default_Swing_03}}
 {id=WavePlayer_2, properties {soundWaveAssetPtr=&Lobster_Default_Swing_01}}
 {id=WavePlayer_20, properties {soundWaveAssetPtr=/Galileo/Items/Lobster/Assets/Sounds/Default/Lobster_Default_Swing_04}}
 {id=WavePlayer_21, properties {soundWaveAssetPtr=/Galileo/Items/Lobster/Assets/Sounds/Default/Lobster_Default_Swing_05}}
 {id=WavePlayer_22, properties {soundWaveAssetPtr=&Lobster_Default_Swing_06}}
 {id=WavePlayer_23, properties {soundWaveAssetPtr=&Lobster_Default_Swing_07}}
 {id=WavePlayer_24, properties {soundWaveAssetPtr=&Lobster_Default_Swing_08}}
 {id=WavePlayer_25, properties {soundWaveAssetPtr=&Lobster_Default_Swing_01}}
 {id=WavePlayer_26, properties {soundWaveAssetPtr=&Lobster_Default_Swing_02}}
 {id=WavePlayer_27, properties {soundWaveAssetPtr=&Lobster_Default_Swing_03}}
 {id=WavePlayer_28, properties {soundWaveAssetPtr=&Lobster_Default_Swing_02}}
 {id=WavePlayer_29, properties {soundWaveAssetPtr=&Lobster_Default_Swing_03}}
 {id=WavePlayer_3, properties {soundWaveAssetPtr=&Lobster_Default_Swing_02}}
 {id=WavePlayer_30, properties {soundWaveAssetPtr=&Lobster_Default_Impact_01}}
 {id=WavePlayer_31, properties {soundWaveAssetPtr=/Galileo/Items/Lobster/Assets/Sounds/Default/Lobster_Default_Impact_02}}
 {id=WavePlayer_32, properties {soundWaveAssetPtr=/Galileo/Items/Lobster/Assets/Sounds/Default/Lobster_Default_Impact_03}}
 {id=WavePlayer_4, properties {soundWaveAssetPtr=&Lobster_Default_Swing_03}}
 {id=WavePlayer_5, properties {soundWaveAssetPtr=&Lobster_Default_Swing_04}}
 {id=WavePlayer_6, properties {soundWaveAssetPtr=&Lobster_Default_Swing_05}}
 {id=WavePlayer_7, properties {soundWaveAssetPtr=&Lobster_Default_Impact_02}}
 {id=WavePlayer_8, properties {soundWaveAssetPtr=&Lobster_Default_Impact_03}}
 {id=WavePlayer_9, properties {soundWaveAssetPtr=/Galileo/Items/Lobster/Assets/Sounds/Default/Lobster_Default_Impact_04}}
]
properties {
 maxDistance=10058.75
 soundClassObject=&Fort_WeaponMelee
 volumeMultiplier=0.5623
 concurrencyOverrides {resolutionRule=StopLowestPriority}
}
