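# Sound cue for the in-air cast loop of the thrown Lobster melee weapon's Realm variant projectile, with attenuation and swing concurrency limits.
# A mixer blends two modulated layers, each passed through a smoothed, low-intensity Doppler node.
# The louder layer is a looping cast wave shaped by an Enveloper with volume and pitch curves.
# The quieter layer randomly picks one of eight non-looping Default swing waves and pitches it up.
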SoundCue /Galileo/Items/Lobster/Assets/Sounds/Realm/Lobster_Realm_Throw_Prj_InAir_Loop_Cast_Cue
root=Mixer_0
nodes [
 {id=Doppler_0, children [Modulator_0], properties {bUseSmoothing=true, dopplerIntensity=0.15}}
 {id=Doppler_1, children [Modulator_1], properties {bUseSmoothing=true, dopplerIntensity=0.15}}
 {
  id=Enveloper_0
  children [WavePlayer_12]
  properties {
   pitchCurve {editorCurveData {keys [
      {arriveTangent=-1.3699, interpMode=RCIM_Cubic, leaveTangent=-1.3699, tangentMode=RCTM_User, time=0}
      {arriveTangent=-1.6004, interpMode=RCIM_Cubic, leaveTangent=-1.6004, tangentMode=RCTM_User, time=0.1, value=1.3}
      {arriveTangent=0.0242, interpMode=RCIM_Cubic, leaveTangent=0.0242, tangentMode=RCTM_User, time=1}
      <at 1.5 0.9>
     ]}}
   volumeCurve {editorCurveData {keys [<at 0 0 RCIM_Cubic>,<at 0.25 1>,<at 0.2866 1>,<at 2 0.6>]}}
  }
 }
 {id=Mixer_0, children [Doppler_1,Doppler_0], properties {inputVolume [0.4,0.8]}}
 {id=Modulator_0, children [Enveloper_0]}
 {id=Modulator_1, children [Random_0], properties {pitchMax=1.45, pitchMin=1.4, volumeMax=0.6, volumeMin=0.5}}
 {
  id=Random_0
  children [WavePlayer_0,WavePlayer_1,WavePlayer_2,WavePlayer_3,WavePlayer_4,WavePlayer_5,WavePlayer_6,WavePlayer_7]
  properties {bShouldExcludeFromBranchCulling=true, weights [1,1,1,1,1,1,1,1]}
 }
 {id=WavePlayer_0, properties {soundWaveAssetPtr=/Galileo/Items/Lobster/Assets/Sounds/Default/Lobster_Default_Swing_08}}
 {id=WavePlayer_1, properties {soundWaveAssetPtr=/Galileo/Items/Lobster/Assets/Sounds/Default/Lobster_Default_Swing_01}}
 {id=WavePlayer_12, properties {bLooping=true, soundWaveAssetPtr=&/Lobster_Realm_Throw_Prj_Cast_Loop}}
 {id=WavePlayer_2, properties {soundWaveAssetPtr=/Galileo/Items/Lobster/Assets/Sounds/Default/Lobster_Default_Swing_02}}
 {id=WavePlayer_3, properties {soundWaveAssetPtr=/Galileo/Items/Lobster/Assets/Sounds/Default/Lobster_Default_Swing_03}}
 {id=WavePlayer_4, properties {soundWaveAssetPtr=/Galileo/Items/Lobster/Assets/Sounds/Default/Lobster_Default_Swing_04}}
 {id=WavePlayer_5, properties {soundWaveAssetPtr=/Galileo/Items/Lobster/Assets/Sounds/Default/Lobster_Default_Swing_05}}
 {id=WavePlayer_6, properties {soundWaveAssetPtr=/Galileo/Items/Lobster/Assets/Sounds/Default/Lobster_Default_Swing_06}}
 {id=WavePlayer_7, properties {soundWaveAssetPtr=/Galileo/Items/Lobster/Assets/Sounds/Default/Lobster_Default_Swing_07}}
]
properties {
 attenuationSettings=/Game/Sounds/Attenuation/Fort_Weapon_Melee_Lobster_Prj_InAir_Loop_Att
 maxDistance=6000
 soundClassObject=&Fort_WeaponMelee
 volumeMultiplier=0.5623
 concurrencyOverrides {resolutionRule=StopLowestPriority}
 concurrencySet [/Galileo/Items/Lobster/Assets/Sounds/Lobster_Swings_OwnerLimit_Conc, /Galileo/Items/Lobster/Assets/Sounds/Lobster_Swings_GlobalLimit_Conc]
}
