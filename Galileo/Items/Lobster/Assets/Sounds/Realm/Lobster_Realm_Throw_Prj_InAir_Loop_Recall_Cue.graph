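# Sound cue for recalling the thrown Realm-variant Lobster melee weapon projectile while it is in the air.
# A mixer layers a briefly delayed random default swing, a random recall pull and a looping Realm sound shaped by pitch and volume envelopes.
# An IsLocal branch plays the pull louder for the local player than otherwise, with pitch randomized.
# Projectile loop attenuation and a concurrency override limit its reach and instance count.
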SoundCue /Galileo/Items/Lobster/Assets/Sounds/Realm/Lobster_Realm_Throw_Prj_InAir_Loop_Recall_Cue
root=Mixer_0
nodes [
 {id=Branch_0, children [Modulator_1,Modulator_2,Modulator_2], properties {boolParameterName=IsLocal}}
 {id=Delay_0, children [Modulator_0], properties {delayMax=0.2, delayMin=0.2}}
 {
  id=Enveloper_1
  children [Modulator_3]
  properties {
   pitchCurve {editorCurveData {keys [<at 0 1.1>,<at 0.5 1.25>]}}
   volumeCurve {editorCurveData {keys [
      <at 0 0 RCIM_Cubic>
      {arriveTangent=-0.0839, interpMode=RCIM_Cubic, leaveTangent=-0.0839, tangentMode=RCTM_User, time=0.5, value=1}
      <at 10 0.5>
     ]}}
  }
 }
 {id=Mixer_0, children [Delay_0,Branch_0,Enveloper_1], properties {inputVolume [1,1.3,1]}}
 {id=Modulator_0, children [Random_1], properties {pitchMax=1.2, pitchMin=1.1}}
 {id=Modulator_1, children [Random_0], properties {pitchMax=1.3, pitchMin=1, volumeMax=1.5, volumeMin=1.5}}
 {id=Modulator_2, children [Random_0], properties {pitchMax=1.3, pitchMin=1, volumeMax=0.5, volumeMin=0.5}}
 {id=Modulator_3, children [WavePlayer_2], properties {pitchMax=1, pitchMin=1, volumeMax=1, volumeMin=1}}
 {id=Random_0, children [WavePlayer_0, WavePlayer_1, WavePlayer_12], properties {bShouldExcludeFromBranchCulling=true, weights [1,1,1]}}
 {
  id=Random_1
  children [WavePlayer_3, WavePlayer_4, WavePlayer_5, WavePlayer_6, WavePlayer_7, WavePlayer_8, WavePlayer_9, WavePlayer_10]
  properties {bShouldExcludeFromBranchCulling=true, weights [1,1,1,1,1,1,1,1]}
 }
 {id=WavePlayer_0, properties {soundWaveAssetPtr=&/Lobster_Realm_Throw_Prj_Recall_Pull_01}}
 {id=WavePlayer_1, properties {soundWaveAssetPtr=&/Lobster_Realm_Throw_Prj_Recall_Pull_02}}
 {id=WavePlayer_10, properties {soundWaveAssetPtr=/Galileo/Items/Lobster/Assets/Sounds/Default/Lobster_Default_Swing_07}}
 {id=WavePlayer_12, properties {soundWaveAssetPtr=&/Lobster_Realm_Throw_Prj_Recall_Pull_03}}
 {id=WavePlayer_2, properties {bLooping=true, soundWaveAssetPtr=&/Lobster_Realm_Loop}}
 {id=WavePlayer_3, properties {soundWaveAssetPtr=/Galileo/Items/Lobster/Assets/Sounds/Default/Lobster_Default_Swing_08}}
 {id=WavePlayer_4, properties {soundWaveAssetPtr=/Galileo/Items/Lobster/Assets/Sounds/Default/Lobster_Default_Swing_01}}
 {id=WavePlayer_5, properties {soundWaveAssetPtr=/Galileo/Items/Lobster/Assets/Sounds/Default/Lobster_Default_Swing_02}}
 {id=WavePlayer_6, properties {soundWaveAssetPtr=/Galileo/Items/Lobster/Assets/Sounds/Default/Lobster_Default_Swing_03}}
 {id=WavePlayer_7, properties {soundWaveAssetPtr=/Galileo/Items/Lobster/Assets/Sounds/Default/Lobster_Default_Swing_04}}
 {id=WavePlayer_8, properties {soundWaveAssetPtr=/Galileo/Items/Lobster/Assets/Sounds/Default/Lobster_Default_Swing_05}}
 {id=WavePlayer_9, properties {soundWaveAssetPtr=/Galileo/Items/Lobster/Assets/Sounds/Default/Lobster_Default_Swing_06}}
]
properties {
 attenuationSettings=/Game/Sounds/Attenuation/Fort_Weapon_Melee_Lobster_Prj_InAir_Loop_Att
 bOverrideConcurrency=true
 maxDistance=6000
 soundClassObject=&Fort_WeaponMelee
 volumeMultiplier=0.5
 concurrencyOverrides {maxCount=4, resolutionRule=StopLowestPriority}
}
