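# Sound cue for the recall whoosh of the Lobster Realm thrown melee weapon projectile, with separate local and non-local mixes.
# An IsLocal branch sends local playback through a quiet, higher-pitched modulated mix, while false or unset gets a boosted mix with smoothed, mild Doppler.
# Each path randomly picks one of four whoosh variations, under a root pitch modulator.
# Melee projectile attenuation applies, and overridden concurrency caps playback at two instances.
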SoundCue /Galileo/Items/Lobster/Assets/Sounds/Realm/Lobster_Realm_Throw_Prj_Recall_Whoosh_Cue
root=Modulator_0
nodes [
 {id=Branch_0, children [Modulator_1,Doppler_0,Doppler_0], properties {boolParameterName=IsLocal}}
 {id=Doppler_0, children [Mixer_1], properties {bUseSmoothing=true, dopplerIntensity=0.2}}
 {id=Mixer_0, children [Random_0], properties {inputVolume [0.15]}}
 {id=Mixer_1, children [Random_1], properties {inputVolume [1.5]}}
 {id=Modulator_0, children [Branch_0], properties {pitchMax=1.1}}
 {id=Modulator_1, children [Mixer_0], properties {pitchMax=1.2, pitchMin=1.1, volumeMax=1, volumeMin=0.9}}
 {
  id=Random_0
  children [WavePlayer_0,WavePlayer_1,WavePlayer_2,WavePlayer_3]
  properties {bShouldExcludeFromBranchCulling=true, weights [1,1,1,1]}
 }
 {
  id=Random_1
  children [WavePlayer_4,WavePlayer_5,WavePlayer_6,WavePlayer_7]
  properties {bShouldExcludeFromBranchCulling=true, weights [1,1,1,1]}
 }
 {id=WavePlayer_0, properties {soundWaveAssetPtr=&/Realm_Throw_Recall_Whoosh_01}}
 {id=WavePlayer_1, properties {soundWaveAssetPtr=&/Realm_Throw_Recall_Whoosh_02}}
 {id=WavePlayer_2, properties {soundWaveAssetPtr=&/Realm_Throw_Recall_Whoosh_03}}
 {id=WavePlayer_3, properties {soundWaveAssetPtr=&/Realm_Throw_Recall_Whoosh_04}}
 {id=WavePlayer_4, properties {soundWaveAssetPtr=&Realm_Throw_Recall_Whoosh_01}}
 {id=WavePlayer_5, properties {soundWaveAssetPtr=&Realm_Throw_Recall_Whoosh_02}}
 {id=WavePlayer_6, properties {soundWaveAssetPtr=&Realm_Throw_Recall_Whoosh_03}}
 {id=WavePlayer_7, properties {soundWaveAssetPtr=&Realm_Throw_Recall_Whoosh_04}}
]
properties {
 attenuationSettings=/Game/Sounds/Attenuation/Fort_Weapon_Melee_Lobster_Prj_InAir_Whip_Att
 bOverrideConcurrency=true
 maxDistance=6000
 soundClassObject=&Fort_WeaponMelee
 volumeMultiplier=1
 concurrencyOverrides {maxCount=2, resolutionRule=StopLowestPriority}
}
