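# Sound cue for the Lobster item's Realm throw release, played as a close-range melee weapon sound.
# A mixer layers a random Default swing whoosh with a louder, fixed Realm throw warning wave.
# A quality-level switch gives the first two levels eight swing variations and the last two a four-wave subset.
# A concurrency override caps it at two instances, stopping the lowest priority.
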SoundCue /Galileo/Items/Lobster/Assets/Sounds/Realm/Lobster_Realm_Throw_Release_Cue
root=Mixer_0
nodes [
 {id=Mixer_0, children [QualityLevel_0, WavePlayer_13], properties {inputVolume [1,1.5]}}
 {id=QualityLevel_0, children [Random_0,Random_0,Random_4,Random_4]}
 {
  id=Random_0
  children [WavePlayer_0,WavePlayer_1,WavePlayer_2,WavePlayer_3,WavePlayer_4,WavePlayer_5,WavePlayer_6,WavePlayer_7]
  properties {bShouldExcludeFromBranchCulling=true, weights [1,1,1,1,1,1,1,1]}
 }
 {
  id=Random_4
  children [WavePlayer_8, WavePlayer_9, WavePlayer_10, WavePlayer_11]
  properties {bShouldExcludeFromBranchCulling=true, weights [1,1,1,1]}
 }
 {id=WavePlayer_0, properties {soundWaveAssetPtr=/Galileo/Items/Lobster/Assets/Sounds/Default/Lobster_Default_Swing_08}}
 {id=WavePlayer_1, properties {soundWaveAssetPtr=/Galileo/Items/Lobster/Assets/Sounds/Default/Lobster_Default_Swing_01}}
 {id=WavePlayer_10, properties {soundWaveAssetPtr=/Galileo/Items/Lobster/Assets/Sounds/Default/Lobster_Default_Swing_02}}
 {id=WavePlayer_11, properties {soundWaveAssetPtr=/Galileo/Items/Lobster/Assets/Sounds/Default/Lobster_Default_Swing_03}}
 {id=WavePlayer_13, properties {soundWaveAssetPtr=&/Realm_Throw_Warning_01}}
 {id=WavePlayer_2, properties {soundWaveAssetPtr=&Lobster_Default_Swing_02}}
 {id=WavePlayer_3, properties {soundWaveAssetPtr=&Lobster_Default_Swing_03}}
 {id=WavePlayer_4, properties {soundWaveAssetPtr=/Galileo/Items/Lobster/Assets/Sounds/Default/Lobster_Default_Swing_04}}
 {id=WavePlayer_5, properties {soundWaveAssetPtr=/Galileo/Items/Lobster/Assets/Sounds/Default/Lobster_Default_Swing_05}}
 {id=WavePlayer_6, properties {soundWaveAssetPtr=/Galileo/Items/Lobster/Assets/Sounds/Default/Lobster_Default_Swing_06}}
 {id=WavePlayer_7, properties {soundWaveAssetPtr=/Galileo/Items/Lobster/Assets/Sounds/Default/Lobster_Default_Swing_07}}
 {id=WavePlayer_8, properties {soundWaveAssetPtr=&Lobster_Default_Swing_08}}
 {id=WavePlayer_9, properties {soundWaveAssetPtr=&Lobster_Default_Swing_01}}
]
properties {
 attenuationSettings=/Game/Sounds/Attenuation/Fort_Weapon_Melee_Attack_Close_Range_Stereo_Att
 bOverrideConcurrency=true
 maxDistance=10058.75
 soundClassObject=&Fort_WeaponMelee
 volumeMultiplier=1
 concurrencyOverrides {maxCount=2, resolutionRule=StopLowestPriority}
}
