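# Sound cue that plays the looping Realm leap sound for the Lobster item, classed as melee weapon audio.
# A looping wave player feeds an Enveloper that fades volume in over the first half second, then a Modulator with its volume range pinned at unity.
# The cue uses the Realm special attack attenuation and overrides concurrency to one instance, stopping the lowest priority.
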SoundCue /Galileo/Items/Lobster/Assets/Sounds/Realm/Realm_Leap_Loop_Cue
root=Modulator_1
nodes [
 {
  id=Enveloper_0
  children [WavePlayer_1]
  properties {pitchCurve {editorCurveData {keys [{time=0}]}}, volumeCurve {editorCurveData {keys [<at 0 0 RCIM_Cubic>,<at 0.5 1>]}}}
 }
 {id=Modulator_1, children [Enveloper_0], properties {volumeMax=1, volumeMin=1}}
 {id=WavePlayer_1, properties {bLooping=true, soundWaveAssetPtr=&/Realm_Leap_Loop_01}}
]
properties {
 attenuationSettings=&/Realm_Special_Attack_Att
 bOverrideConcurrency=true
 maxDistance=16000
 soundClassObject=&Fort_WeaponMelee
 volumeMultiplier=1
 concurrencyOverrides {maxCount=1, resolutionRule=StopLowestPriority}
}
