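# Sound cue for the start of the Realm leap on the Lobster item, routed to the melee weapon sound class with special-attack attenuation.
# A mixer layers the begin wave, a quieter copy of it at raised pitch under an enveloper's rising pitch curve, and a louder sweetener.
# The sweetener is randomly either a pitched-down first variant or an unmodified second one.
# A concurrency override allows one instance, stopping the lowest priority.
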SoundCue /Galileo/Items/Lobster/Assets/Sounds/Realm/Realm_Leap_Start_Cue
root=Mixer_0
nodes [
 {
  id=Enveloper_0
  children [Modulator_2]
  properties {pitchCurve {editorCurveData {keys [{time=0}, {time=0}, <at 0.4 2.5>]}}, volumeCurve {editorCurveData {keys [<at 0 1>]}}}
 }
 {id=Mixer_0, children [WavePlayer_0,Enveloper_0,Random_1], properties {inputVolume [1,0.5,2]}}
 {id=Modulator_2, children [WavePlayer_2], properties {pitchMax=1.5, pitchMin=1.5, volumeMax=1, volumeMin=1}}
 {id=Modulator_3, children [WavePlayer_3], properties {pitchMax=0.8, pitchMin=0.7, volumeMax=1, volumeMin=1}}
 {id=Random_1, children [Modulator_3,WavePlayer_4], properties {weights [1,1]}}
 {id=WavePlayer_0, properties {soundWaveAssetPtr=&/Realm_Leap_Begin_01}}
 {id=WavePlayer_2, properties {soundWaveAssetPtr=&Realm_Leap_Begin_01}}
 {id=WavePlayer_3, properties {soundWaveAssetPtr=&/Realm_Leap_Begin_Sweetener_01}}
 {id=WavePlayer_4, properties {soundWaveAssetPtr=&/Realm_Leap_Begin_Sweetener_02}}
]
properties {
 attenuationSettings=&/Realm_Special_Attack_Att
 bOverrideConcurrency=true
 maxDistance=16000
 soundClassObject=&Fort_WeaponMelee
 volumeMultiplier=1.4
 concurrencyOverrides {maxCount=1, resolutionRule=StopLowestPriority}
}
