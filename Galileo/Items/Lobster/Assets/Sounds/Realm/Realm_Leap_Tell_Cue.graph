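# Sound cue for the Realm leap tell in the Lobster item set, using the melee weapon sound class.
# A single wave player feeds a modulator that randomly lowers pitch while holding volume constant.
# The cue boosts overall volume and uses special attack attenuation.
# Its concurrency override allows only one instance at a time, stopping the lowest priority one.
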SoundCue /Galileo/Items/Lobster/Assets/Sounds/Realm/Realm_Leap_Tell_Cue
root=Modulator_0
nodes [
 {id=Modulator_0, children [WavePlayer_0], properties {pitchMax=0.8, pitchMin=0.6, volumeMax=1, volumeMin=1}}
 {id=WavePlayer_0, properties {soundWaveAssetPtr=&/Realm_Leap_Tell_01}}
]
properties {
 attenuationSettings=&/Realm_Special_Attack_Att
 bOverrideConcurrency=true
 maxDistance=16000
 soundClassObject=&Fort_WeaponMelee
 volumeMultiplier=1.8
 concurrencyOverrides {maxCount=1, resolutionRule=StopLowestPriority}
}
