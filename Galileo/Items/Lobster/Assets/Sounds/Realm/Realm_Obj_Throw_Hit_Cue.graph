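# Sound cue for a thrown Realm object's hit, in the melee weapon sound class.
# An IsPawn branch sends non-pawn or unset hits to randomized world impacts at half volume with world attenuation.
# Pawn hits play a random player impact, non-spatialized and unattenuated, only when Islocal is true, and are otherwise silent.
# A concurrency override caps simultaneous instances and limits retriggering, stopping the lowest priority.
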SoundCue /Galileo/Items/Lobster/Assets/Sounds/Realm/Realm_Obj_Throw_Hit_Cue
root=Modulator_0
nodes [
 {
  id=Attenuation_0
  children [Modulator_1]
  properties {attenuationSettings=&/Realm_Obj_Throw_Hit_World_Att}
 }
 {id=Attenuation_1, children [Modulator_2], properties {bOverrideAttenuation=true, attenuationOverrides {bAttenuate=false, bSpatialize=false}}}
 {id=Branch_0, children [Branch_1,Mixer_1,Mixer_1], properties {boolParameterName=IsPawn}}
 {id=Branch_1, children [Attenuation_1, null, null], properties {boolParameterName=Islocal}}
 {id=Mixer_1, children [Attenuation_0], properties {inputVolume [0.5]}}
 {id=Modulator_0, children [Branch_0], properties {volumeMax=1, volumeMin=1}}
 {id=Modulator_1, children [Random_0]}
 {id=Modulator_2, children [Random_1]}
 {
  id=Random_0
  children [WavePlayer_0,WavePlayer_1,WavePlayer_3,WavePlayer_4,WavePlayer_5]
  properties {bShouldExcludeFromBranchCulling=true, weights [1,1,1,1,1]}
 }
 {id=Random_1, children [WavePlayer_6,WavePlayer_7,WavePlayer_8], properties {bShouldExcludeFromBranchCulling=true, weights [1,1,1]}}
 {id=WavePlayer_0, properties {soundWaveAssetPtr=&/Realm_Obj_Throw_Hit_World_05}}
 {id=WavePlayer_1, properties {soundWaveAssetPtr=&/Realm_Obj_Throw_Hit_World_02}}
 {id=WavePlayer_3, properties {soundWaveAssetPtr=&/Realm_Obj_Throw_Hit_World_03}}
 {id=WavePlayer_4, properties {soundWaveAssetPtr=&/Realm_Obj_Throw_Hit_World_04}}
 {id=WavePlayer_5, properties {soundWaveAssetPtr=&/Realm_Obj_Throw_Hit_World_01}}
 {id=WavePlayer_6, properties {soundWaveAssetPtr=&/Realm_Obj_Throw_Hit_Player_01}}
 {id=WavePlayer_7, properties {soundWaveAssetPtr=&/Realm_Obj_Throw_Hit_Player_02}}
 {id=WavePlayer_8, properties {soundWaveAssetPtr=&/Realm_Obj_Throw_Hit_Player_03}}
]
properties {
 bOverrideConcurrency=true
 maxDistance=8000
 soundClassObject=&Fort_WeaponMelee
 volumeMultiplier=1.2
 concurrencyOverrides {maxCount=2, resolutionRule=StopLowestPriority, retriggerTime=0.5}
}
