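# Sound cue named for the launch of a Realm object throw in the Galileo Lobster item set.
# Its node graph is empty with no root, so the file references no audio to play.
# It carries playback settings only: special-attack attenuation, the melee weapon sound class, slightly reduced volume and a long max distance.
# A concurrency override caps simultaneous instances at three, stopping the lowest priority first.
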SoundCue /Galileo/Items/Lobster/Assets/Sounds/Realm/Realm_Obj_Throw_Launch_Cue
root=null
nodes []
properties {
 attenuationSettings=&/Realm_Special_Attack_Att
 bOverrideConcurrency=true
 maxDistance=16000
 soundClassObject=&Fort_WeaponMelee
 volumeMultiplier=0.9
 concurrencyOverrides {maxCount=3, resolutionRule=StopLowestPriority}
}
