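# Sound cue for a thrown Realm object projectile, mixing an in-air loop with a one-shot launch sound, in the melee weapon sound class.
# The looping wave fades in through an Enveloper volume curve, while the launch branch randomly picks one of three waves with pitch variation at a lower mix level.
# Each branch has its own attenuation node, and a concurrency override caps simultaneous instances, stopping the lowest priority.
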SoundCue /Galileo/Items/Lobster/Assets/Sounds/Realm/Realm_Obj_Throw_Prj_InAir_Loop_Cue
root=Mixer_0
nodes [
 {id=Attenuation_0, children [Modulator_1], properties {attenuationSettings=&/Realm_Obj_Throw_Att}}
 {
  id=Attenuation_1
  children [Modulator_0]
  properties {attenuationSettings=&/Realm_Obj_Throw_Prj_Loop_Att}
 }
 {
  id=Enveloper_0
  children [WavePlayer_0]
  properties {
   pitchCurve {editorCurveData {keys [{time=0}]}}
   volumeCurve {editorCurveData {keys [<at 0 0 RCIM_Cubic>, {arriveTangent=3.5757, interpMode=RCIM_Cubic, leaveTangent=3.5757, tangentMode=RCTM_User, time=0.5, value=1}]}}
  }
 }
 {id=Mixer_0, children [Attenuation_1, Attenuation_0], properties {inputVolume [1,0.7]}}
 {id=Modulator_0, children [Enveloper_0], properties {volumeMax=1, volumeMin=1}}
 {id=Modulator_1, children [Random_0], properties {pitchMax=1.2}}
 {id=Random_0, children [WavePlayer_1,WavePlayer_2,WavePlayer_3], properties {bShouldExcludeFromBranchCulling=true, weights [1,1,1]}}
 {id=WavePlayer_0, properties {bLooping=true, soundWaveAssetPtr=&/Realm_Obj_Throw_Prj_InAir_Loop}}
 {id=WavePlayer_1, properties {soundWaveAssetPtr=&/Realm_Obj_Throw_Launch_01}}
 {id=WavePlayer_2, properties {soundWaveAssetPtr=&/Realm_Obj_Throw_Launch_02}}
 {id=WavePlayer_3, properties {soundWaveAssetPtr=&/Realm_Obj_Throw_Launch_03}}
]
properties {
 bOverrideConcurrency=true
 maxDistance=12000
 soundClassObject=&Fort_WeaponMelee
 volumeMultiplier=1.1
 concurrencyOverrides {maxCount=3, resolutionRule=StopLowestPriority}
}
