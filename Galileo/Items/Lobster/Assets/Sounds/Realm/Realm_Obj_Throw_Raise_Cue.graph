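# Sound cue for the raise sound of the Realm object throw on the Lobster item, classed as melee weapon audio.
# It randomly picks one of three equally weighted wave variations, then passes it through a modulator that varies pitch while holding volume steady.
# Playback uses the Realm special attack attenuation, with a concurrency override that caps simultaneous instances and stops the lowest priority one.
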SoundCue /Galileo/Items/Lobster/Assets/Sounds/Realm/Realm_Obj_Throw_Raise_Cue
root=Modulator_0
nodes [
 {id=Modulator_0, children [Random_0], properties {pitchMax=1.3, volumeMax=1, volumeMin=1}}
 {id=Random_0, children [WavePlayer_0,WavePlayer_1,WavePlayer_2], properties {bShouldExcludeFromBranchCulling=true, weights [1,1,1]}}
 {id=WavePlayer_0, properties {soundWaveAssetPtr=&/Realm_Obj_Throw_Raise_01}}
 {id=WavePlayer_1, properties {soundWaveAssetPtr=&/Realm_Obj_Throw_Raise_02}}
 {id=WavePlayer_2, properties {soundWaveAssetPtr=&/Realm_Obj_Throw_Raise_03}}
]
properties {
 attenuationSettings=&/Realm_Special_Attack_Att
 bOverrideConcurrency=true
 maxDistance=16000
 soundClassObject=&Fort_WeaponMelee
 volumeMultiplier=1.2
 concurrencyOverrides {maxCount=2, resolutionRule=StopLowestPriority}
}
