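# Sound cue for the launch of the Realm throw on the Lobster item, classed as melee weapon audio.
# A single launch wave plays through a Modulator that randomizes pitch with a raised upper limit, so repeats vary.
# It uses the Realm special attack attenuation and a boosted cue volume.
# A concurrency override allows two simultaneous instances, stopping the lowest priority one beyond that.
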SoundCue /Galileo/Items/Lobster/Assets/Sounds/Realm/Realm_Throw_Launch_Cue
root=Modulator_0
nodes [
 {id=Modulator_0, children [WavePlayer_0], properties {pitchMax=1.3}}
 {id=WavePlayer_0, properties {soundWaveAssetPtr=&/Realm_Throw_Launch_01}}
]
properties {
 attenuationSettings=&/Realm_Special_Attack_Att
 bOverrideConcurrency=true
 maxDistance=16000
 soundClassObject=&Fort_WeaponMelee
 volumeMultiplier=1.3
 concurrencyOverrides {maxCount=2, resolutionRule=StopLowestPriority}
}
